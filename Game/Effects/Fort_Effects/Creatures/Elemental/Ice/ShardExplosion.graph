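# Lit ice shard explosion material for mesh particles in the elemental creature effects, breaking a mesh into flying pieces.
# SelfBuildingWalls2 supplies per-piece pivots and orientations, while AnimationMF turns Break Apart into a phase that drives world position offset: pieces fly out, spin and shrink to their pivots.
# Fresnel-weighted tint, particle color and a view-aligned crystal reflection form the glow.
# Detail normals blend with UV-encoded base normals, with a correction for the rotation.
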Material /Game/Effects/Fort_Effects/Creatures/Elemental/Ice/ShardExplosion
graph {
 properties {
  bUsedWithMeshParticles=true
  bUsedWithNiagaraMeshParticles=true
  numCustomizedUVs=5
  pixelDepthOffsetMode=PDOM_Legacy
  usedShadingModels=Default Lit
 }
 nodes [
  n1 = MaterialFunctionCall!(Number of sub objects, Manual Anim Grad, Sec mot Falloff rad) {materialFunction=/Game/Effects/Fort_Effects/Materials/Temp/SelfBuildingWalls2}
  n2 = Subtract(A=@n1.Random Per Mesh) {constB=0.5}
  n3 = Frac(None=Add(A=@n1.Random Per Mesh) {constB=0.25})
  n4 = Constant() {r=10}
  n5 = MaterialFunctionCall!(
   TotalNumberOfSubObjects=@n4
   BuildAnimPhase=@n4
   TargetBreakNumber
   BreakAnim
   Code_totalDestructionAmount=ScalarParameter() {defaultValue=0.01, parameterName=Break Apart}
  ) {materialFunction=/Game/Packages/User_Built_Props/Materials/Master/AnimationMF}
  n6 = Multiply(
   A=Multiply(
    A=MaterialFunctionCall!(In=@n2) {materialFunction=/Engine/Functions/Engine_MaterialFunctions02/Utility/Sign}.Result
    B=LinearInterpolate(Alpha=@n3) {constA=0.5, constB=1.25}
   )
   B=@n5.single
  )
  n7 = Normalize(VectorInput=@n1.Board Orientation)
  n8 = Multiply(
   A=Multiply(A=Multiply(A=@n5.single, B=@n7), B=ScalarParameter() {defaultValue=50, parameterName=Flight Distance}) {constB=50, desc=flight}
   B=MaterialFunctionCall() {materialFunction=/Engine/Functions/Engine_MaterialFunctions02/WorldPositionOffset/ObjectScale}.Scale XYZ
  )
  n9 = Add(
   A=RotateAboutAxis(NormalizedRotationAxis=@n1.up Cross Orientation, RotationAngle=@n6, PivotPoint=@n1.up Cross Orientation, Position=WorldPosition().XYZ)
   B=@n8
  )
  n10 = Add(A=@n8, B=MaterialFunctionCall() {materialFunction=&SelfBuildingWalls2}.Board Pivot Pos)
  n11 = Add(A=@n9, B=WorldPosition().XYZ)
  Add(A=@n9, B=Multiply(A=Subtract(A=@n10, B=@n11), B=@n5.single)) {desc="wpo -->"}
  n12 = Fresnel() {baseReflectFraction=0.25, exponent=1}
  n13 = Multiply(
   A=Multiply(A=VectorParameter() {parameterName=Base Color, defaultValue=<rgba 0.2997 0.3778 0.461 1>}.RGB, B=ParticleColor().RGB)
   B=Clamp(None=@n12)
  )
  n14 = TextureObject() {texture=/Game/Environments/Props/Rocks/Textures/T_Props_CrystalEnv_Purple}
  n15 = TextureCoordinate()
  n16 = LinearInterpolate(
   A=Constant3Vector() {constant=<rgba 0 0 1 0>}
   B=TextureSample() {samplerType=SAMPLERTYPE_Normal, texture=/Game/Environments/Props/Rocks/Textures/T_Props_Crystal_N}.RGB
   Alpha=Ceil(None=ComponentMask(None=@n15) {g=true})
  )
  n17 = MaterialFunctionCall(f2=Abs(None=@n16)) {materialFunction=&AddComponents}
  n18 = Multiply(
   A=Multiply(A=MaterialFunctionCall!(ReflectionVector, ReflectionTexture=@n14) {materialFunction=/Engine/Functions/Engine_MaterialFunctions01/Reflections/ViewAlignedReflection}.Texture) {constB=5}
   B=Multiply(A=Power(Base=@n17.f2))
  )
  MaterialFunctionCall(Emissive=Multiply(A=Add(A=@n13, B=@n18)) {constB=1.5}) {materialFunction=&MF_FX_EmissiveAdjustments}
  MaterialFunctionCall(SpecularColor=Multiply(A=OneMinus(None=@n12), B=Fresnel() {baseReflectFraction=0, exponent=1})) {materialFunction=&ConvertFromDiffSpec}
  MaterialFunctionCall(BaseNormal=DeriveNormalZ(InXY=TextureCoordinate() {coordinateIndex=4}), AdditionalNormal=@n16) {materialFunction=&BlendAngleCorrectedNormals}
  MaterialFunctionCall(
   Rotation Axis=MaterialFunctionCall() {materialFunction=&SelfBuildingWalls2}.up Cross Orientation
   Rotation Angle=@n6
  ) {materialFunction=&FixRotateAboutAxisNormals}
  n19 = OneMinus(None=@n12)
  Multiply(B=Multiply(A=@n19, B=@n19)) {constA=0.5, constB=0}
 ]
}
